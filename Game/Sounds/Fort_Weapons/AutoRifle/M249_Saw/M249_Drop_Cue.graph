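# Sound cue for the M249 SAW weapon drop sound, classed under loot collection sounds.
# A single wave player plays one drop sample, with no randomization or layering.
# The cue lowers volume and pitch and uses a shared weapon-drop attenuation.
# It overrides concurrency so that new instances are blocked once the limit is reached, rather than cutting off playing ones.
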SoundCue /Game/Sounds/Fort_Weapons/AutoRifle/M249_Saw/M249_Drop_Cue
root=WavePlayer_0
nodes [{id=WavePlayer_0, properties {soundWaveAssetPtr=&/M249_Drop_01}}]
properties {
 attenuationSettings=&Fort_Weapon_Drop_Att
 bOverrideConcurrency=true
 maxDistance=3000
 pitchMultiplier=0.9
 soundClassObject=&Fort_Loot_Collect
 volumeMultiplier=0.6031
 concurrencyOverrides {resolutionRule=PreventNew}
}
