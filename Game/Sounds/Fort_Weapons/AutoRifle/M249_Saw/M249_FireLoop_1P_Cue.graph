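# Sound cue for the looping first-person fire sound of the M249 SAW, an automatic rifle weapon.
# A single looping close-fire wave plays through a Modulator that caps pitch slightly below normal.
# It uses first-person attenuation, a ranged-weapon fire sound class and a shared first-person gunfire concurrency set, with the cue's debug flag left on.
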SoundCue /Game/Sounds/Fort_Weapons/AutoRifle/M249_Saw/M249_FireLoop_1P_Cue
root=Modulator_0
nodes [
 {id=Modulator_0, children [WavePlayer_1], properties {pitchMax=0.95}}
 {id=WavePlayer_1, properties {bLooping=true, soundWaveAssetPtr=&/M249_Close_FireLoop_01}}
]
properties {
 attenuationSettings=&Fort_Weapon_Projectile_FirstPerson_Athena_Att
 bDebug=true
 maxDistance=2097152
 soundClassObject=&Fort_WeaponRanged_Fire_1P_PSM
 volumeMultiplier=0.8
 concurrencyOverrides {bLimitToOwner=true, maxCount=3, resolutionRule=StopOldest}
 concurrencySet [&Weapon_Gunfire_FirstPerson_Conc]
}
