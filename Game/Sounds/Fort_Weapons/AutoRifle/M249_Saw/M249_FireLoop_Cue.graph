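# Sound cue for the looping automatic fire of the M249 SAW, layered by listener distance.
# Close, medium and distant looping waves each pass through a modulator, distance-specific attenuation and sound class override into a weighted mixer.
# Close and medium use the third-person fire class, distant has its own, and the medium and distant waves are boosted in volume.
# Local and global gunfire concurrency sets limit overlapping instances.
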SoundCue /Game/Sounds/Fort_Weapons/AutoRifle/M249_Saw/M249_FireLoop_Cue
root=Mixer_17
nodes [
 {
  id=Attenuation_0
  children [Modulator_1]
  properties {attenuationSettings=&Fort_Weapon_Projectile_Medium_Mono_Athena_Att}
 }
 {
  id=Attenuation_34
  children [Modulator_12]
  properties {attenuationSettings=&Fort_Weapon_Projectile_Distant_Mono_Athena_Att}
 }
 {id=Attenuation_49, children [Modulator_0], properties {attenuationSettings=&Fort_Weapon_Projectile_Close_Athena_Att}}
 {id=Mixer_17, children [SoundClass_32, SoundClass_0, SoundClass_35], properties {inputVolume [0.8,0.6,0.45]}}
 {id=Modulator_0, children [WavePlayer_1], properties {pitchMax=0.95}}
 {id=Modulator_1, children [WavePlayer_2], properties {pitchMax=0.95, volumeMax=2.05, volumeMin=1.95}}
 {id=Modulator_12, children [WavePlayer_4], properties {pitchMax=0.95, volumeMax=2.05, volumeMin=1.95}}
 {
  id=SoundClass_0
  children [Attenuation_0]
  properties {soundClassOverride=&Fort_WeaponRanged_Fire_3P_PSM}
 }
 {
  id=SoundClass_32
  children [Attenuation_49]
  properties {soundClassOverride=&Fort_WeaponRanged_Fire_3P_PSM}
 }
 {
  id=SoundClass_35
  children [Attenuation_34]
  properties {soundClassOverride=&Fort_WeaponRanged_Fire_Distant}
 }
 {id=WavePlayer_1, properties {bLooping=true, soundWaveAssetPtr=&/M249_Close_FireLoop_01}}
 {id=WavePlayer_2, properties {bLooping=true, soundWaveAssetPtr=&/M249_MediumNew_FireLoop_01}}
 {id=WavePlayer_4, properties {bLooping=true, soundWaveAssetPtr=&/M249_DistantNew_FireLoop_01}}
]
properties {
 bDebug=true
 maxDistance=27500
 volumeMultiplier=0.8
 concurrencyOverrides {bLimitToOwner=true, maxCount=3, resolutionRule=StopOldest}
 concurrencySet [&Weapon_Gunfire_Local_Conc, &Weapon_Gunfire_Global_Conc]
}
