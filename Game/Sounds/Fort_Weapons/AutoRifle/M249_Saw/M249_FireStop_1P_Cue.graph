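# First-person sound cue for the end of M249 SAW automatic fire, layering a close fire-stop sound with shell casing audio.
# A mixer plays the modulated fire-stop wave alongside a quieter modulated branch that picks one of four casing waves at random after a fixed short delay.
# The casing branch overrides attenuation with logarithmic falloff, while the cue uses first-person weapon attenuation, sound class and gunfire concurrency assets.
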SoundCue /Game/Sounds/Fort_Weapons/AutoRifle/M249_Saw/M249_FireStop_1P_Cue
root=Mixer_0
nodes [
 {
  id=Attenuation_1
  children [Delay_0]
  properties {bOverrideAttenuation=true, attenuationOverrides {distanceAlgorithm=Logarithmic, falloffDistance=1400, attenuationShapeExtents {x=200}}}
 }
 {id=Delay_0, children [Random_0], properties {delayMax=0.2, delayMin=0.2}}
 {id=Mixer_0, children [Modulator_0,Modulator_2], properties {inputVolume [1,0.4]}}
 {id=Modulator_0, children [WavePlayer_1]}
 {id=Modulator_2, children [Attenuation_1]}
 {id=Random_0, children [WavePlayer_5,WavePlayer_6,WavePlayer_7,WavePlayer_8], properties {weights [1,1,1,1]}}
 {id=WavePlayer_1, properties {soundWaveAssetPtr=&/M249_Close_FireStop_01}}
 {id=WavePlayer_5, properties {soundWaveAssetPtr=&/M249_Casings_01}}
 {id=WavePlayer_6, properties {soundWaveAssetPtr=&/M249_Casings_02}}
 {id=WavePlayer_7, properties {soundWaveAssetPtr=&/M249_Casings_03}}
 {id=WavePlayer_8, properties {soundWaveAssetPtr=&/M249_Casings_04}}
]
properties {
 attenuationSettings=&Fort_Weapon_Projectile_FirstPerson_Athena_Att
 bDebug=true
 maxDistance=2097152
 soundClassObject=&Fort_WeaponRanged_Fire_1P_PSM
 volumeMultiplier=0.8
 concurrencyOverrides {bLimitToOwner=true, maxCount=3, resolutionRule=StopOldest}
 concurrencySet [&Weapon_Gunfire_FirstPerson_Conc]
}
