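# Sound cue for the M249 SAW fire-stop gunshot tail, mixing close, medium and distant layers, each with its own attenuation.
# The close layer also carries a quieter, delayed random pick of four shell-casing sounds with an overridden short falloff.
# Close and medium layers use the third-person fire sound class, while distant uses the distant class with raised, randomized pitch.
# Local and global gunfire concurrency sets apply, and the debug flag is on.
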SoundCue /Game/Sounds/Fort_Weapons/AutoRifle/M249_Saw/M249_FireStop_Cue
root=Mixer_17
nodes [
 {
  id=Attenuation_0
  children [Modulator_1]
  properties {attenuationSettings=&Fort_Weapon_Projectile_Medium_Mono_Athena_Att}
 }
 {
  id=Attenuation_1
  children [Delay_0]
  properties {bOverrideAttenuation=true, attenuationOverrides {distanceAlgorithm=Logarithmic, falloffDistance=1400, attenuationShapeExtents {x=200}}}
 }
 {
  id=Attenuation_34
  children [Modulator_12]
  properties {attenuationSettings=&Fort_Weapon_Projectile_Distant_Mono_Athena_Att}
 }
 {
  id=Attenuation_49
  children [WavePlayer_1]
  properties {attenuationSettings=&Fort_Weapon_Projectile_Close_Athena_Att}
 }
 {id=Delay_0, children [Random_0], properties {delayMax=0.2, delayMin=0.2}}
 {id=Mixer_0, children [Modulator_0,Modulator_2], properties {inputVolume [1,0.4]}}
 {id=Mixer_17, children [SoundClass_32, SoundClass_0, SoundClass_35], properties {inputVolume [0.8,0.7,0.45]}}
 {id=Modulator_0, children [Attenuation_49]}
 {id=Modulator_1, children [WavePlayer_2], properties {volumeMax=2.05, volumeMin=1.95}}
 {id=Modulator_12, children [WavePlayer_4], properties {pitchMax=1.2, pitchMin=1.1, volumeMax=2.05, volumeMin=1.95}}
 {id=Modulator_2, children [Attenuation_1]}
 {id=Random_0, children [WavePlayer_5,WavePlayer_6,WavePlayer_7,WavePlayer_8], properties {weights [1,1,1,1]}}
 {
  id=SoundClass_0
  children [Attenuation_0]
  properties {soundClassOverride=&Fort_WeaponRanged_Fire_3P_PSM}
 }
 {id=SoundClass_32, children [Mixer_0], properties {soundClassOverride=&Fort_WeaponRanged_Fire_3P_PSM}}
 {
  id=SoundClass_35
  children [Attenuation_34]
  properties {soundClassOverride=&Fort_WeaponRanged_Fire_Distant}
 }
 {id=WavePlayer_1, properties {soundWaveAssetPtr=&/M249_Close_FireStop_01}}
 {id=WavePlayer_2, properties {soundWaveAssetPtr=&/M249_MediumNew_FireStop_01}}
 {id=WavePlayer_4, properties {soundWaveAssetPtr=&/M249_DistantNew_FireStop_01}}
 {id=WavePlayer_5, properties {soundWaveAssetPtr=&/M249_Casings_01}}
 {id=WavePlayer_6, properties {soundWaveAssetPtr=&/M249_Casings_02}}
 {id=WavePlayer_7, properties {soundWaveAssetPtr=&/M249_Casings_03}}
 {id=WavePlayer_8, properties {soundWaveAssetPtr=&/M249_Casings_04}}
]
properties {
 bDebug=true
 maxDistance=27500
 volumeMultiplier=0.8
 concurrencyOverrides {bLimitToOwner=true, maxCount=3, resolutionRule=StopOldest}
 concurrencySet [&Weapon_Gunfire_Local_Conc, &Weapon_Gunfire_Global_Conc]
}
