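# First-person fire sound cue for the Quartz variant of the M249 SAW.
# A Mixer layers random shots, slightly delayed pitched-down mechanical sounds, later shell casings, and a low-ammo layer.
# A continuous modulator on AmmoNormalized fades the low-ammo layer in as ammo runs low, under its LowAmmo_Critical sound class.
# The other layers use the 3P fire class despite the 1p cue; concurrency is owner-limited, stopping the oldest.
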SoundCue /Game/Sounds/Fort_Weapons/AutoRifle/M249_Saw/M249_Quartz_Fire_Cue_1p
root=Mixer_0
nodes [
 {id=Delay_0, children [SoundClass_2], properties {delayMax=0.01, delayMin=0.01}}
 {id=Delay_1, children [SoundClass_0], properties {delayMax=0.35, delayMin=0.35}}
 {id=Mixer_0, children [SoundClass_3,Delay_1,Delay_0,SoundClass_1], properties {inputVolume [1,1,1,1]}}
 {
  id=ModulatorContinuous_2
  children [WavePlayer_11]
  properties {volumeModulationParams {default=0, maxInput=0.25, maxOutput=0, minInput=0.2, minOutput=1.5, parameterName=AmmoNormalized}}
 }
 {id=Modulator_0, children [Random_2], properties {pitchMax=0.8, pitchMin=0.8, volumeMax=2, volumeMin=2}}
 {id=Modulator_1, children [Random_1], properties {volumeMax=1.5, volumeMin=1.5}}
 {id=Modulator_4, children [Random_0], properties {volumeMax=0.5, volumeMin=0.5}}
 {id=Modulator_6, children [ModulatorContinuous_2], properties {pitchMax=0.8, pitchMin=0.8, volumeMax=0.95}}
 {id=Random_0, children [WavePlayer_6,WavePlayer_7,WavePlayer_8,WavePlayer_9], properties {weights [1,1,1,1]}}
 {
  id=Random_1
  children [
   WavePlayer_0
   WavePlayer_1
   WavePlayer_2
   WavePlayer_3
   WavePlayer_4
   WavePlayer_5
   WavePlayer_23
   WavePlayer_24
   WavePlayer_25
   WavePlayer_26
  ]
  properties {weights [1,1,1,1,1,1,1,1,1,1]}
 }
 {
  id=Random_2
  children [WavePlayer_13, WavePlayer_14, WavePlayer_15, WavePlayer_16, WavePlayer_17, WavePlayer_18]
  properties {weights [1,1,1,1,1,1]}
 }
 {id=SoundClass_0, children [Modulator_4], properties {soundClassOverride=&Fort_WeaponRanged_Fire_3P_PSM}}
 {
  id=SoundClass_1
  children [Modulator_6]
  properties {soundClassOverride=/Game/Sounds/FortSoundClassesAndModes/Fort_WeaponRanged_LowAmmo_Critical_PSM}
 }
 {id=SoundClass_2, children [Modulator_0], properties {soundClassOverride=&Fort_WeaponRanged_Fire_3P_PSM}}
 {id=SoundClass_3, children [Modulator_1], properties {soundClassOverride=&Fort_WeaponRanged_Fire_3P_PSM}}
 {id=WavePlayer_0, properties {soundWaveAssetPtr=&/M249_Quartz_Shoot_01}}
 {id=WavePlayer_1, properties {soundWaveAssetPtr=&/M249_Quartz_Shoot_02}}
 {id=WavePlayer_11, properties {soundWaveAssetPtr=&/M249_Quartz_LowAmmo}}
 {id=WavePlayer_13, properties {soundWaveAssetPtr=&/M249_Quartz_Mech_01}}
 {id=WavePlayer_14, properties {soundWaveAssetPtr=&/M249_Quartz_Mech_02}}
 {id=WavePlayer_15, properties {soundWaveAssetPtr=&/M249_Quartz_Mech_03}}
 {id=WavePlayer_16, properties {soundWaveAssetPtr=&/M249_Quartz_Mech_04}}
 {id=WavePlayer_17, properties {soundWaveAssetPtr=&/M249_Quartz_Mech_05}}
 {id=WavePlayer_18, properties {soundWaveAssetPtr=&/M249_Quartz_Mech_06}}
 {id=WavePlayer_2, properties {soundWaveAssetPtr=&/M249_Quartz_Shoot_03}}
 {id=WavePlayer_23, properties {soundWaveAssetPtr=&/M249_Quartz_Shoot_07}}
 {id=WavePlayer_24, properties {soundWaveAssetPtr=&/M249_Quartz_Shoot_08}}
 {id=WavePlayer_25, properties {soundWaveAssetPtr=&/M249_Quartz_Shoot_09}}
 {id=WavePlayer_26, properties {soundWaveAssetPtr=&/M249_Quartz_Shoot_10}}
 {id=WavePlayer_3, properties {soundWaveAssetPtr=&/M249_Quartz_Shoot_04}}
 {id=WavePlayer_4, properties {soundWaveAssetPtr=&/M249_Quartz_Shoot_05}}
 {id=WavePlayer_5, properties {soundWaveAssetPtr=&/M249_Quartz_Shoot_06}}
 {id=WavePlayer_6, properties {soundWaveAssetPtr=&/M249_Casings_01}}
 {id=WavePlayer_7, properties {soundWaveAssetPtr=&/M249_Casings_02}}
 {id=WavePlayer_8, properties {soundWaveAssetPtr=&/M249_Casings_03}}
 {id=WavePlayer_9, properties {soundWaveAssetPtr=&/M249_Casings_04}}
]
properties {
 attenuationSettings=&Fort_Weapon_Projectile_FirstPerson_Athena_Att
 bDebug=true
 maxDistance=2097152
 pitchMultiplier=1.15
 volumeMultiplier=0.45
 concurrencyOverrides {bLimitToOwner=true, maxCount=2, resolutionRule=StopOldest}
 concurrencySet [&Weapon_Gunfire_FirstPerson_Conc]
}
