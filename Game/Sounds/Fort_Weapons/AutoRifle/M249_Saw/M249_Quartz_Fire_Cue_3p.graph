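# Third-person fire sound cue for the M249 SAW Quartz weapon, layering close, medium and distant shot recordings.
# Each layer randomly picks from ten wave variations, then passes through a volume boost, distance-specific attenuation and a sound class override.
# Close and medium layers share a sub-mix, which a root mixer blends with the quieter distant layer.
# The cue raises pitch and limits concurrency per owner, stopping the oldest, alongside shared gunfire concurrency sets.
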SoundCue /Game/Sounds/Fort_Weapons/AutoRifle/M249_Saw/M249_Quartz_Fire_Cue_3p
root=Mixer_2
nodes [
 {id=Attenuation_3, children [Modulator_1], properties {attenuationSettings=&Fort_Weapon_Projectile_Close_Athena_Att}}
 {
  id=Attenuation_4
  children [Modulator_0]
  properties {attenuationSettings=&Fort_Weapon_Projectile_Distant_Mono_Athena_Att}
 }
 {
  id=Attenuation_5
  children [Modulator_2]
  properties {attenuationSettings=&Fort_Weapon_Projectile_Medium_Mono_Athena_Att}
 }
 {id=Mixer_2, children [Modulator_3,SoundClass_4], properties {inputVolume [1,0.35]}}
 {id=Mixer_3, children [SoundClass_3,SoundClass_5], properties {inputVolume [0.85,0.6]}}
 {id=Modulator_0, children [Random_5], properties {volumeMax=2, volumeMin=2}}
 {id=Modulator_1, children [Random_1], properties {volumeMax=2, volumeMin=2}}
 {id=Modulator_2, children [Random_4], properties {volumeMax=3.5, volumeMin=3.5}}
 {id=Modulator_3, children [Mixer_3], properties {pitchMax=1, pitchMin=1}}
 {
  id=Random_1
  children [
   WavePlayer_30
   WavePlayer_31
   WavePlayer_32
   WavePlayer_33
   WavePlayer_34
   WavePlayer_35
   WavePlayer_36
   WavePlayer_37
   WavePlayer_38
   WavePlayer_39
  ]
  properties {weights [1,1,1,1,1,1,1,1,1,1]}
 }
 {
  id=Random_4
  children [
   WavePlayer_6
   WavePlayer_7
   WavePlayer_8
   WavePlayer_9
   WavePlayer_10
   WavePlayer_11
   WavePlayer_12
   WavePlayer_13
   WavePlayer_14
   WavePlayer_15
  ]
  properties {weights [1,1,1,1,1,1,1,1,1,1]}
 }
 {
  id=Random_5
  children [
   WavePlayer_16
   WavePlayer_17
   WavePlayer_18
   WavePlayer_19
   WavePlayer_20
   WavePlayer_21
   WavePlayer_22
   WavePlayer_27
   WavePlayer_28
   WavePlayer_29
  ]
  properties {weights [1,1,1,1,1,1,1,1,1,1]}
 }
 {
  id=SoundClass_3
  children [Attenuation_3]
  properties {soundClassOverride=&Fort_WeaponRanged_Fire_3P_PSM}
 }
 {
  id=SoundClass_4
  children [Attenuation_4]
  properties {soundClassOverride=&Fort_WeaponRanged_Fire_Distant}
 }
 {
  id=SoundClass_5
  children [Attenuation_5]
  properties {soundClassOverride=&Fort_WeaponRanged_Fire_Distant}
 }
 {id=WavePlayer_10, properties {soundWaveAssetPtr=&/M249_Quartz_Shoot_Medium_05}}
 {id=WavePlayer_11, properties {soundWaveAssetPtr=&/M249_Quartz_Shoot_Medium_06}}
 {id=WavePlayer_12, properties {soundWaveAssetPtr=&/M249_Quartz_Shoot_Medium_07}}
 {id=WavePlayer_13, properties {soundWaveAssetPtr=&/M249_Quartz_Shoot_Medium_08}}
 {id=WavePlayer_14, properties {soundWaveAssetPtr=&/M249_Quartz_Shoot_Medium_09}}
 {id=WavePlayer_15, properties {soundWaveAssetPtr=&/M249_Quartz_Shoot_Medium_10}}
 {id=WavePlayer_16, properties {soundWaveAssetPtr=&/M249_Quartz_Shoot_Distant_01}}
 {id=WavePlayer_17, properties {soundWaveAssetPtr=&/M249_Quartz_Shoot_Distant_02}}
 {id=WavePlayer_18, properties {soundWaveAssetPtr=&/M249_Quartz_Shoot_Distant_03}}
 {id=WavePlayer_19, properties {soundWaveAssetPtr=&/M249_Quartz_Shoot_Distant_04}}
 {id=WavePlayer_20, properties {soundWaveAssetPtr=&/M249_Quartz_Shoot_Distant_05}}
 {id=WavePlayer_21, properties {soundWaveAssetPtr=&/M249_Quartz_Shoot_Distant_06}}
 {id=WavePlayer_22, properties {soundWaveAssetPtr=&/M249_Quartz_Shoot_Distant_07}}
 {id=WavePlayer_27, properties {soundWaveAssetPtr=&/M249_Quartz_Shoot_Distant_08}}
 {id=WavePlayer_28, properties {soundWaveAssetPtr=&/M249_Quartz_Shoot_Distant_09}}
 {id=WavePlayer_29, properties {soundWaveAssetPtr=&/M249_Quartz_Shoot_Distant_10}}
 {id=WavePlayer_30, properties {soundWaveAssetPtr=&/M249_Quartz_Shoot_Close_01}}
 {id=WavePlayer_31, properties {soundWaveAssetPtr=&/M249_Quartz_Shoot_Close_02}}
 {id=WavePlayer_32, properties {soundWaveAssetPtr=&/M249_Quartz_Shoot_Close_03}}
 {id=WavePlayer_33, properties {soundWaveAssetPtr=&/M249_Quartz_Shoot_Close_04}}
 {id=WavePlayer_34, properties {soundWaveAssetPtr=&/M249_Quartz_Shoot_Close_05}}
 {id=WavePlayer_35, properties {soundWaveAssetPtr=&/M249_Quartz_Shoot_Close_06}}
 {id=WavePlayer_36, properties {soundWaveAssetPtr=&/M249_Quartz_Shoot_Close_07}}
 {id=WavePlayer_37, properties {soundWaveAssetPtr=&/M249_Quartz_Shoot_Close_08}}
 {id=WavePlayer_38, properties {soundWaveAssetPtr=&/M249_Quartz_Shoot_Close_09}}
 {id=WavePlayer_39, properties {soundWaveAssetPtr=&/M249_Quartz_Shoot_Close_10}}
 {id=WavePlayer_6, properties {soundWaveAssetPtr=&/M249_Quartz_Shoot_Medium_01}}
 {id=WavePlayer_7, properties {soundWaveAssetPtr=&/M249_Quartz_Shoot_Medium_02}}
 {id=WavePlayer_8, properties {soundWaveAssetPtr=&/M249_Quartz_Shoot_Medium_03}}
 {id=WavePlayer_9, properties {soundWaveAssetPtr=&/M249_Quartz_Shoot_Medium_04}}
]
properties {
 bDebug=true
 maxDistance=27500
 pitchMultiplier=1.15
 volumeMultiplier=0.6
 concurrencyOverrides {bLimitToOwner=true, maxCount=2, resolutionRule=StopOldest}
 concurrencySet [&Weapon_Gunfire_Global_Conc, &Weapon_Gunfire_Local_Conc]
}
