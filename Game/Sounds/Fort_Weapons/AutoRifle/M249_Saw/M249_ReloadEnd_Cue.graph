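# Sound cue for the end-of-reload foley of the M249 SAW automatic rifle.
# A single wave player plays the M249_ReloadEnd_01 wave at reduced volume, with no randomization or mixing.
# It uses the shared weapon reload attenuation, the weapon foley sound class and an M249-specific source effect chain.
# Concurrency is overridden on the cue, and its debug flag is left on.
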SoundCue /Game/Sounds/Fort_Weapons/AutoRifle/M249_Saw/M249_ReloadEnd_Cue
root=WavePlayer_1
nodes [{id=WavePlayer_1, properties {soundWaveAssetPtr=&/M249_ReloadEnd_01}}]
properties {
 attenuationSettings=&Fort_Weapon_Reload_Att
 bDebug=true
 bOverrideConcurrency=true
 maxDistance=2250
 soundClassObject=&Fort_WeaponFoley
 sourceEffectChain=&/M249FX_Preset1
 volumeMultiplier=0.662
}
